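# First-person low-health gameplay cue effect for Athena: a Niagara system with one burst-spawned sprite emitter that uses fixed facing.
# The sprite persists past its lifetime, with alpha following the system Timeline's normalized time and size set from the system's Base Scale.
# Dynamic material parameters include one animated by a waveform through a curve.
# User parameters expose a fade toggle, FOV and aspect ratio.
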NiagaraSystem /FNGameplayCues/Athena/Effects/Particles/Gameplay/LowHealth/NS_FP_LowHealth
dependencies {
 usedDataInterfaces [&NiagaraDataInterfaceCurve]
 usedDynamicInputs [&Multiply_Vector2D]
 usedModules [
  &SystemState
  &Timeline
  &self:SystemSpawnScript.SystemScriptSource.SystemScriptGraph.NiagaraNodeAssignment_0.SetVariables_#1
 ]
}
emitters {Minimal {
  emitterUpdateScript [
   &EmitterState()
   &SpawnBurst_Instantaneous(Use Spawn Probability=false, UseLoopCountLimit=true)
  ]
  particleSpawnScript [
   &InitializeParticle(
    Color=<rgba 2 2 2 1>
    Sprite Size Mode=::NewEnumerator4
    Sprite UV Mode=::NewEnumerator3
    UsePositionOffset=false
    Write Material Random=true
   )
   &SpriteFacingAndAlignment(Sprite Alignment=<0 0 1>, Sprite Facing=<-1 0 0>, WriteAlignment=true, WriteFacing=true)
  ]
  particleUpdateScript [
   &ParticleState(Kill Particles When Lifetime Has Elapsed=false)
   &ScaleColor(
    Scale Alpha=&FloatFromCurve(
     FloatCurve=DataInterface() {propertyValues {Curve {keys [<at 0 0 RCIM_Cubic>,<at 1 1 RCIM_Cubic>]}}}
     CurveIndex=$System.Timeline.NormalizedTime
    )
    ScaleA=true
    ScaleRGB=true
   ) {module=ScaleColor001}
   &DynamicMaterialParameters(
    Index 0 Param 1=5
    Index 0 Param 2=0.1
    Index 0 Param 3=1.25
    Index 0 Param 4=&FloatFromCurve(
     FloatCurve=DataInterface() {propertyValues {Curve {keys [<at 0 0.1665>,<at 1 0.333>]}}}
     CurveIndex=&Waveform("Waveform [1]"=::NewEnumerator2, "[1] Frequency"=1, "[1] Compound Frequency"=1, "[1] Amplitude Min/Max"=<0 1>)
     Scale Curve=2
    )
    Module.Index0Param0Write=true
    Module.Index0Param1Write=true
    Module.Index0Param2Write=true
    Module.Index0Param3Write=true
   )
   /SetVariables(Particles.SpriteSize=$System.Base Scale) {isSetParameters=true}
  ]
  rendererClasses [&NiagaraSpriteRendererProperties]
  renderers [{rendererClass=&NiagaraSpriteRendererProperties}]
 }}
summary {
 systemName=NS_FP_LowHealth
 emitters [{emitterName=Minimal, rendererClasses [&NiagaraSpriteRendererProperties]}]
 userVariables [
  User.FadeInOut {type=NiagaraBool, defaultValue=true}
  User.FOV {type=NiagaraFloat, defaultValue=0, description=""}
  User.Ratio {type=Vector2f, description="", defaultValue {struct=Vector2f, value=<1.7778 1>}}
 ]
}
